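# Animation Blueprint for the Tower Sentinel Female pickaxe that adds procedural idle wobble to its bones.
# AmbientMotion turns game time into per-axis sine translation and rotation offsets.
# Each update calls it four times with differing speed and magnitude, storing results that additively drive bones local_1 to local_4 in bone space over the input pose.
# A second Transform Bone chain for chain_1 to chain_4 is not connected to the output.
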AnimBlueprint /BRCosmetics/Weapons/FORT_Melee/Pickaxe_Tower_Sentinel_Female/Meshes/Tower_Sentinel_Female_AnimBP
target_skeleton=&/Tower_Sentinel_Female_Skeleton
graphs [
 AmbientMotion {nodes [
   n1 = AmbientMotion!() -> (
    then
    Period:Vector="1.5,1,2"
    PeriodMult:double=1
    TimeOffset:double=~
    Magnitude:Vector="0.3,0.56,0.22"
    MagnitudeMult:double=1
    RotateMult:double=1
   ) {is=title K2Node_FunctionEntry}
   n2 = BreakVector(InVec=@n1.Magnitude)
   n3 = BreakVector(InVec=@n1)
   n4 = GetGameTimeInSeconds()
   n5 = float * float(A=float * float(A=@n1.PeriodMult, B=@n3.Y), B=@n4)
   n6 = float * float(A="Sin (Radians)"(A=float + float(A=@n5, B=@n1.TimeOffset) {is=title K2Node_CommutativeAssociativeBinaryOperator}), B=@n2.Y)
   n7 = float * float(A=float * float(A=@n1.PeriodMult, B=@n3.Z), B=@n4)
   n8 = float * float(A="Sin (Radians)"(A=float + float(A=@n7, B=@n1.TimeOffset)), B=@n2.Z)
   n9 = float * float(A=float * float(A=@n1.PeriodMult, B=@n3), B=@n4)
   n10 = float * float(A="Sin (Radians)"(A=float + float(A=@n9, B=@n1.TimeOffset)), B=@n2)
   n11 = MakeRotator(Roll=@n10, Pitch=@n6, Yaw=@n8)
   > Return Node!(
    execute
    VectorOut:Vector=vector * float(A=MakeVector(X=@n10, Y=@n6, Z=@n8), B=@n1.MagnitudeMult)
    RotateOut:Rotator="Scale Rotator (Float)"!(A:Rotator="Scale Rotator (Float)"(A=@n11, B=@n1.MagnitudeMult), B:float=@n1.RotateMult) -> (ReturnValue:Rotator)
   )
  ]}
 AnimGraph {nodes [
   n1 = "Transform (Modify) Bone - Bone: chain_1"(Alpha=1, Translation=Get ambientTransA() -> (ambientTransA:Vector), Rotation=Get ambientRotA() -> (ambientRotA:Rotator), Scale="1,1,1") {translationSpace=BCS_ParentBoneSpace, alphaBoolBlend {blendOption=Linear}, boneToModify {boneName=chain_1}}
   n2 = Get ambientRotB() -> (ambientRotB:Rotator)
   n3 = Get ambientTransB() -> (ambientTransB:Vector)
   "Transform (Modify) Bone - Bone: chain_4"(
    ComponentPose="Transform (Modify) Bone - Bone: chain_3"(
     ComponentPose="Transform (Modify) Bone - Bone: chain_2"(ComponentPose=@n1, Alpha=1, Translation=@n3, Rotation=@n2, Scale="1,1,1") {translationSpace=BCS_ParentBoneSpace, alphaBoolBlend {blendOption=Linear}, boneToModify {boneName=chain_2}}
     Alpha=1
     Translation=Get ambientTransC() -> (ambientTransC:Vector)
     Rotation=Get ambientRotC() -> (ambientRotC:Rotator)
     Scale="1,1,1"
    ) {translationSpace=BCS_ParentBoneSpace, alphaBoolBlend {blendOption=Linear}, boneToModify {boneName=chain_3}}
    Alpha=1
    Translation=Get ambientTransD() -> (ambientTransD:Vector)
    Rotation=Get ambientRotD() -> (ambientRotD:Rotator)
    Scale="1,1,1"
   ) {translationSpace=BCS_ParentBoneSpace, alphaBoolBlend {blendOption=Linear}, boneToModify {boneName=chain_4}}
   AnimGraphNode_SubInput_0 = Input Pose()
   n4 = "Transform (Modify) Bone - Bone: local_1"(
    ComponentPose=Local To Component(LocalPose=@AnimGraphNode_SubInput_0)
    Alpha=1
    Translation=Get ambientTransA()
    Rotation=Get ambientRotA()
    Scale="1,1,1"
   ) {
    rotationMode=BMM_Additive
    rotationSpace=BCS_BoneSpace
    translationMode=BMM_Additive
    translationSpace=BCS_BoneSpace
    alphaBoolBlend {blendOption=Linear}
    boneToModify {boneName=local_1}
   }
   n5 = "Transform (Modify) Bone - Bone: local_2"(ComponentPose=@n4, Alpha=1, Translation=Get ambientTransB(), Rotation=Get ambientRotB(), Scale="1,1,1") {
    rotationMode=BMM_Additive
    rotationSpace=BCS_BoneSpace
    translationMode=BMM_Additive
    translationSpace=BCS_BoneSpace
    alphaBoolBlend {blendOption=Linear}
    boneToModify {boneName=local_2}
   }
   n6 = "Transform (Modify) Bone - Bone: local_3"(ComponentPose=@n5, Alpha=1, Translation=Get ambientTransC(), Rotation=Get ambientRotC(), Scale="1,1,1") {
    rotationMode=BMM_Additive
    rotationSpace=BCS_BoneSpace
    translationMode=BMM_Additive
    translationSpace=BCS_BoneSpace
    alphaBoolBlend {blendOption=Linear}
    boneToModify {boneName=local_3}
   }
   n7 = "Transform (Modify) Bone - Bone: local_4"(ComponentPose=@n6, Alpha=1, Translation=Get ambientTransD(), Rotation=Get ambientRotD(), Scale="1,1,1") {
    rotationMode=BMM_Additive
    rotationSpace=BCS_BoneSpace
    translationMode=BMM_Additive
    translationSpace=BCS_BoneSpace
    alphaBoolBlend {blendOption=Linear}
    boneToModify {boneName=local_4}
   }
   Output Pose(Result=Component To Local(ComponentPose=@n7)) {layerGroup=None}
  ]}
 EventGraph {nodes [
   Event BlueprintUpdateAnimation()
   > n1 = Sequence!(execute) -> (then_0, then_1, then_2:Exec, then_3:Exec)
   n2 = Self-Reference()
   > n3 = AmbientMotion!(
    execute
    self=@n2
    Period:Vector="1.5,1,2"
    PeriodMult:double=0.9
    TimeOffset:double=0
    Magnitude:Vector="0.7,0.56,0.22"
    MagnitudeMult:double=0.7
    RotateMult:double=10
   ) -> (then, VectorOut:Vector, RotateOut:Rotator) {is=title K2Node_CallFunction}
   > Set ambientTransA(ambientTransA:Vector=@n3) -> (Output_Get:Vector)
   > Set ambientRotA(ambientRotA:Rotator=@n3.RotateOut) -> (Output_Get:Rotator)
   n4 = AmbientMotion(execute=@n1.then_1, self=@n2, Period="1.5,1,2", PeriodMult=0.7, TimeOffset=0, Magnitude="0.7,0.56,0.22", MagnitudeMult=0.85, RotateMult=12)
   > Set ambientTransB(ambientTransB:Vector=@n4) -> (Output_Get:Vector)
   > Set ambientRotB(ambientRotB:Rotator=@n4.RotateOut) -> (Output_Get:Rotator)
   n5 = AmbientMotion(execute=@n1.then_2, self=@n2, Period="1.5,1,2", PeriodMult=0.5, TimeOffset=0, Magnitude="0.7,0.56,0.22", MagnitudeMult=0.8, RotateMult=14)
   > Set ambientTransC(ambientTransC:Vector=@n5) -> (Output_Get:Vector)
   > Set ambientRotC(ambientRotC:Rotator=@n5.RotateOut) -> (Output_Get:Rotator)
   n6 = AmbientMotion(execute=@n1.then_3, self=@n2, Period="1.5,1,2", PeriodMult=0.3, TimeOffset=0, Magnitude="0.7,0.56,0.22", MagnitudeMult=0.5, RotateMult=30)
   > Set ambientTransD(ambientTransD:Vector=@n6) -> (Output_Get:Vector)
   > Set ambientRotD(ambientRotD:Rotator=@n6.RotateOut) -> (Output_Get:Rotator)
   TryGetPawnOwner()
  ]}
]
variables [
 ambientRotA {type=Rotator}
 ambientRotB {type=Rotator}
 ambientRotC {type=Rotator}
 ambientRotD {type=Rotator}
 ambientTransA {type=Vector}
 ambientTransB {type=Vector}
 ambientTransC {type=Vector}
 ambientTransD {type=Vector}
]
